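# Translucent unlit material for an NPC interaction "OutOfLoop" effect, drawing a glitchy, flickering, tinted shell on meshes.
# World-aligned noise, gated by position-offset sines, plus a TurnOn-driven height pulse pushes vertices horizontally outward, with optional vertex-color hiding.
# Digital noise and sine-banded, distance-adjusted Fresnel drive hue-shifted colors into eye-adaptation-corrected emissive.
# Fresnel terms shape opacity, masked against custom depth on deferred to fix translucent self-sorting.
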Material /Game/Effects/Fort_Effects/Gameplay/NPC/M_NPC_Interaction_OutOfLoop
graph {
 properties {
  allowTranslucentCustomDepthWrites=true
  bEnableResponsiveAA=true
  bUsedWithClothing=true
  bUsedWithInstancedStaticMeshes=true
  bUsedWithMeshDeformer=true
  bUsedWithMorphTargets=true
  bUsedWithSkeletalMesh=true
  bUsedWithSplineMeshes=true
  blendMode=BLEND_Translucent
  opacityMaskClipValue=0
  pixelDepthOffsetMode=PDOM_Legacy
  shadingModel=MSM_Unlit
  usedShadingModels=Unlit
 }
 nodes [
  n1 = WorldPosition()
  n2 = ObjectPositionWS()
  n3 = Multiply(A=Normalize(VectorInput=Subtract(A=@n1.XYZ, B=@n2)), B=Constant3Vector() {constant=<rgba 1 1 0 0>})
  n4 = Add(
   A=ScalarParameter() {defaultValue=10, group=Noise, parameterName=WPOStrength}
   B=Multiply(A=ObjectRadius(), B=ScalarParameter() {defaultValue=0.025, group=Noise, parameterName=WPOStrengthFromObjectRadius}) {constB=0.05}
  )
  n5 = TextureObject() {samplerType=SAMPLERTYPE_LinearColor, texture=&T_NoisePack_LinearColor}
  n6 = Divide(A=ObjectRadius(), B=ScalarParameter() {defaultValue=0.1, group=Noise, parameterName=NoiseSize})
  n7 = Multiply(A=VectorParameter() {group=Noise, parameterName=NoiseTileMult, defaultValue=<rgba 1 1 0.1 1>}.RGB, B=@n6)
  n8 = Time()
  n9 = VectorParameter() {group=Noise, parameterName=NoisePanSpeed, defaultValue=<rgba -0.08 0.07 -0.2 0>}
  n10 = Add(A=WorldPosition().XYZ, B=Multiply(A=@n6, B=Multiply(A=@n8, B=@n9.RGB)))
  n11 = Multiply(
   A=Power(Base=MaterialFunctionCall(TextureObject=@n5, TextureSize=@n7, WorldPosition=@n10) {materialFunction=&WorldAlignedTexture}.XY Texture) {constExponent=4}
   B=ScalarParameter() {defaultValue=2, group=Noise, parameterName=NoiseMaskMult}
  ) {constB=20}
  n12 = ActorPositionWS()
  n13 = MaterialFunctionCall(Float2=Round(None=Divide(A=@n12) {constB=2000})) {materialFunction=&BreakOutFloat2Components}
  n14 = Add(A=Time(), B=Add(A=@n13.R, B=@n13.R))
  n15 = Multiply(A=ScalarParameter() {defaultValue=3, parameterName=FlickerSpeed}, B=@n14)
  n16 = Sine(None=@n15) {period=1.6}
  n17 = Sine(None=@n15) {period=1.42}
  n18 = Ceil(None=AppendVector(A=AppendVector(A=@n16, B=@n17), B=Sine(None=@n15) {period=1.51}))
  n19 = ScalarParameter() {defaultValue=0.5, group=Noise, parameterName=NoiseFrequency}
  n20 = Add(A=Sine(None=@n15) {period=0.72}, B=Multiply(B=@n19) {constA=0.75}) {constB=0.75}
  n21 = Multiply(A=Add(A=Sine(None=@n15) {period=2.1}, B=Multiply(B=@n19) {constA=0.25}), B=@n20)
  n22 = Multiply(B=@n19) {constA=0.9, constB=0.9}
  n23 = Sine(None=@n15) {period=1.25}
  n24 = Multiply(A=@n20, B=Ceil(None=Add(A=@n22, B=@n23) {constB=0.9}))
  n25 = Sine(None=@n14) {period=0.075}
  n26 = Multiply(A=Saturate(None=Multiply(A=@n21, B=@n24)), B=LinearInterpolate(Alpha=ConstantBiasScale(None=@n25)) {constA=0.7})
  n27 = Add(
   A=ScalarParameter() {group=Noise, parameterName=NoiseAdd}
   B=Multiply(A=Multiply(A=@n18, B=@n26), B=ScalarParameter() {defaultValue=2, group=Noise, parameterName=NoiseMult})
  )
  n28 = Desaturation(None=Saturate(None=Multiply(A=@n11, B=@n27)))
  n29 = ComponentMask(None=WorldPosition().XYZ) {b=true}
  n30 = ComponentMask(None=MaterialFunctionCall() {materialFunction=&ObjectPivotPoint}.Object Pivot Location) {b=true}
  n31 = ScalarParameter() {defaultValue=1, parameterName=TurnOn}
  n32 = LinearInterpolate(A=@n30, B=Add(A=@n30, B=ScalarParameter() {defaultValue=300, parameterName=PulseHeight}) {constB=1600}, Alpha=@n31) {constA=-0.25, constB=1.2}
  n33 = ScalarParameter() {defaultValue=33, parameterName=PulseRadius}
  n34 = LinearInterpolate(A=@n33, B=Multiply(A=@n33) {constB=0.1}, Alpha=@n31) {constA=500, constB=100}
  n35 = Multiply(A=Multiply(A=SphereMask(A=@n29, B=@n32, Radius=@n34) {hardnessPercent=0}, B=OneMinus(None=@n31))) {constB=2}
  Add(
   A=Multiply(
    A=MaterialFunctionCall!(WPO) {materialFunction=/Game/Packages/Fortress_SharedMaterials/Base_Material_Functions/Fortnite_HideElementsWithVertexColor}.Result
    B=ScalarParameter() {parameterName=UseVertexColor}
   )
   B=Multiply(A=Add(A=@n28, B=@n35), B=Multiply(A=@n3, B=@n4)) {constB=0}
  )
  n36 = DynamicParameter() {paramNames [X_position,Y_position,Width0,Width1]}
  n37 = Multiply(A=@n36.Y_position) {constB=-1}
  n38 = Time()
  n39 = Frac(None=Multiply(A=@n38))
  n40 = OneMinus(None=TextureSample(UVs=AppendVector(A=@n38, B=@n39)) {texture=&T_NoisePack_01}.G)
  n41 = Panner(Coordinate=TextureCoordinate()) {speedX=0.1, speedY=1}
  n42 = ComponentMask(None=VectorParameter() {parameterName="Base_texture(R)", defaultValue=<rgba 0 0 0 1>}.RGB) {g=true, r=true}
  n43 = TextureSample(UVs=Add(A=LinearInterpolate(A=@n37, B=@n36.Y_position, Alpha=@n40), B=Multiply(A=@n41, B=@n42))) {samplerType=SAMPLERTYPE_LinearColor, texture=/Game/Global/Textures/Default/Noise/T_Atlas_DigitalNoise_01}
  n44 = LinearInterpolate(
   A=VectorParameter() {parameterName=ColorDarkA, defaultValue=<rgba 0 0.001 0.001 0>}.RGB
   B=VectorParameter() {parameterName=ColorDarkB, defaultValue=<rgba 0.001 0.0001 0 0>}.RGB
   Alpha=@n43.R
  )
  n45 = VectorParameter() {parameterName=ColorOverallTint, defaultValue=<rgba 0.6408 0.8232 1 1>}
  n46 = LinearInterpolate(A=ScalarParameter() {parameterName=ColorSwithRangeMin}, B=ScalarParameter() {parameterName=ColorSwithRangeMax}, Alpha=@n43.R) {constA=-0.05, constB=0.05}
  n47 = ScalarParameter() {defaultValue=5000, parameterName=EdgeThicknessDistanceLerp}
  n48 = Fresnel(ExponentIn=LinearInterpolate(
    A=ScalarParameter() {defaultValue=1.5, parameterName=EdgeThicknessNearest}
    B=ScalarParameter() {defaultValue=0.1, parameterName=EdgeThicknessFarest}
    Alpha=MaterialFunctionCall(Distance=@n47) {materialFunction=&Near_Camera_Fade}.Result
   )) {baseReflectFraction=0, exponent=2.5}
  n49 = Add(
   A=Multiply(A=Power(Base=@n48) {constExponent=0.2}, B=ScalarParameter() {defaultValue=5, parameterName=SineTlie}) {constB=3}
   B=LinearInterpolate(B=ScalarParameter() {defaultValue=1, parameterName=PulseAmount}, Alpha=@n40)
  )
  n50 = MaterialFunctionCall(
   A=VectorParameter() {parameterName=ColorBrightA, defaultValue=<rgba 0.0626 0.7686 1 1>}.RGB
   B=Constant3Vector()
   C=VectorParameter() {parameterName=ColorBrightB, defaultValue=<rgba 1 0 0.02 1>}.RGB
   Alpha=MaterialFunctionCall(In=Sine(None=@n49), Contrast=ScalarParameter() {defaultValue=-0.8, parameterName=Contrast}) {materialFunction=&CheapContrast}.Result
  ) {materialFunction=&Lerp_3Color}
  n51 = Multiply(
   A=Desaturation(
    None=MaterialFunctionCall(Hue Shift Percentage=@n46, Texture=@n50.Result) {materialFunction=&HueShift}.Result
    Fraction=ScalarParameter() {defaultValue=0.2, parameterName=Desaturation}
   )
   B=Power(Base=@n48)
  ) {constA=1}
  MaterialFunctionCall!(
   Emissive=Multiply(A=LinearInterpolate(A=@n44, B=@n45.RGB, Alpha=@n51), B=ScalarParameter() {defaultValue=20, parameterName=Brightness})
   Alpha=ScalarParameter() {defaultValue=1, group=Cosmetic Emissive EyeAdaptation, parameterName=Cosmetic Emissive EyeAdaptation Alpha}
   Material
  ) {materialFunction=/Game/Packages/Fortress_SharedMaterials/Base_Material_Functions/BaseCharacterShaderFunctions/MF_CosmeticEmissiveEyeAdaptation}
  n52 = Fresnel(BaseReflectFractionIn=ScalarParameter() {parameterName=InnerOpacity}) {baseReflectFraction=0.2, exponent=1}
  n53 = Saturate(None=LinearInterpolate(Alpha=Fresnel() {baseReflectFraction=0, exponent=1}) {constA=5, constB=-0.4})
  n54 = Constant()
  n55 = If(
   A=Add(A=PixelDepth().R, B=ScalarParameter() {defaultValue=-0.25, parameterName=BiasSorting_Far})
   B=ComponentMask(None=SceneTexture() {sceneTextureId=PPI_CustomDepth}.Color) {r=true}
   "A > B"=@n54
   "A == B"=@n54
   "A < B"=Constant() {r=1}
  )
  n56 = Constant() {r=1}
  Multiply(
   A=Multiply(A=Multiply(A=@n52, B=@n53), B=ScalarParameter() {defaultValue=1.3, parameterName=OpacityOverall}) {constB=2}
   B=ShadingPathSwitch(Default=@n55, Deferred=@n55, Forward=@n56, Mobile=@n56)
  )
 ]
}
